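; BP_Frosty_Progression_Master is a Blueprint for the Frosty progression effect, apparently a base class for snow-related progression actors.
; Its custom event Snow_SpawnCheck prints the actor's primary asset name and world location as a debug string.
; The construction script only calls its parent, so the file itself holds no other logic.
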(event Custom|Snow_SpawnCheck
  (bind (_primaryassettype _primaryassetname) (Utilities|Struct|BreakPrimaryAssetId (AssetManager|GetPrimaryAssetIdfromObject self)))
  (Development|PrintString (Utilities|String|Append (Utilities|String|ToString(Name) _primaryassetname) " - " (Utilities|String|ToString(Vector) (Transformation|GetActorLocation)))))

(fn ConstructionScript
  (|CallParentFunction))
